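# Niagara system for the wind-up effect of the SonnyBoy player emote: an energy orb that builds, then collapses over a shared effect duration.
# EnergyGuts spins three staggered meshes at the owner offset, while EnrgyGlow layers socket-following sprites, a mesh and a blue light.
# Sparks orbit on slightly tilted rings that tighten, with sprites aligned to motion.
# System loop age drives the collapse, and a time-dilatation alpha speeds up spin and custom timers.
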NiagaraSystem /BRCosmetics/Animation/Game/MainPlayer/Emotes/SonnyBoy/FX/NS_Emote_SonnyBoy_WindUp
dependencies {
 usedDataInterfaces [&NiagaraDataInterfaceColorCurve, &NiagaraDataInterfaceCurve, &NiagaraDataInterfaceSocketReader]
 usedDynamicInputs []
 usedModules [
  &self:SystemSpawnScript.SystemScriptSource.SystemScriptGraph.NiagaraNodeAssignment_1.SetVariables_#1
  &self:SystemSpawnScript.SystemScriptSource.SystemScriptGraph.NiagaraNodeAssignment_2.SetVariables_#2
  &self:SystemSpawnScript.SystemScriptSource.SystemScriptGraph.NiagaraNodeAssignment_3.SetVariables_#3
  &SystemState
  &self:SystemSpawnScript.SystemScriptSource.SystemScriptGraph.NiagaraNodeAssignment_0.SetVariables_#4
  &self:SystemSpawnScript.SystemScriptSource.SystemScriptGraph.NiagaraNodeAssignment_6.SetVariables_#5
 ]
}
emitters {
 NE_EnergyGuts {
  emitterUpdateScript [
   &EmitterState(
    Inactive Response=::NewEnumerator1
    Life Cycle Mode=::NewEnumerator1
    Loop Behavior=::NewEnumerator1
    Loop Duration Mode=::NewEnumerator1
    UseLoopDelay=false
   )
   &SpawnBurst_Instantaneous(Spawn Count=3, Use Spawn Probability=false, UseLoopCountLimit=false)
  ]
  particleSpawnScript [
   &InitializeParticle(
    Color Mode=::NewEnumerator0
    Lifetime=&Multiply_Float(
     A=$System.EffectDuration
     B=&Lerp_Float(A=1, B=1, Alpha=&ReturnNormalizedExecIndex())
    )
    Mesh Scale=<1.5 1.5 0.5>
    Mesh Scale Mode=::NewEnumerator4
    Position Offset=<0 0 80>
    UsePositionOffset=true
    Write Material Random=true
   )
   /SetVariables(Particles.Age:NiagaraFloat=&Lerp_Float(A=-1, B=0, Alpha=&ReturnNormalizedExecIndex())) {isSetParameters=true}
   &InitialMeshOrientation(
    Mesh Orientation Mode=::NewEnumerator7
    Rotation=&Add_Vector(
     A=&Lerp_Vector(A=<0 0 0>, B=<0.05 0 0>, Alpha=&ReturnNormalizedExecIndex())
     B=&Lerp_Vector(A=<0 0 1>, B=<0 0 -1>, Alpha=&ReturnNormalizedExecIndex(Normalized Index Scale=0.3))
    )
    Use Rotation Vector=true
   )
  ]
  particleUpdateScript [
   &ParticleState()
   &ScaleMeshSize(Scale Factor=&Lerp_Vector(
     A=<0 0 0>
     B=<1 1 1>
     Alpha=&Multiply_Float(
      A=&FloatFromCurve(FloatCurve=DataInterface() {propertyValues {Curve {keys [
           {arriveTangent=-0.3977, interpMode=RCIM_Cubic, leaveTangent=-0.3977, tangentMode=RCTM_User, time=0, value=1.4}
           {arriveTangent=-0.3546, interpMode=RCIM_Cubic, leaveTangent=-0.3546, tangentMode=RCTM_User, time=0.85, value=1}
          ]}}})
      B=&FloatFromCurve(
       FloatCurve=DataInterface() {propertyValues {Curve {keys [
           {arriveTangent=0.0389, interpMode=RCIM_Cubic, leaveTangent=0.0389, tangentMode=RCTM_User, time=0, value=1}
           {arriveTangent=-0.3586, interpMode=RCIM_Cubic, leaveTangent=-0.3586, tangentMode=RCTM_User, time=0.85, value=0.7}
           {arriveTangent=-3.5635, interpMode=RCIM_Cubic, leaveTangent=-3.5635, tangentMode=RCTM_User, time=1, value=0}
          ]}}}
       CurveIndex=$System.NormalizedLoopAge
      )
     )
    ))
   /SetVariables(Particles.Position=&AddVectorToPosition(Position=$Engine.Owner.Position, Vector=$System.SystemOffset)) {isSetParameters=true}
   &UpdateMeshOrientation(
    Rotation Rate=&Add_Float(
     A=&Lerp_Float(A=0.1, B=1, Alpha=$System.TimeDilatationAlpha)
     B=&Lerp_Float(
      A=2
      B=0
      Alpha=&FloatFromCurve(
       FloatCurve=DataInterface() {propertyValues {Curve {keys [
           {arriveTangent=0.0389, interpMode=RCIM_Cubic, leaveTangent=0.0389, tangentMode=RCTM_User, time=0, value=1}
           {arriveTangent=-0.3586, interpMode=RCIM_Cubic, leaveTangent=-0.3586, tangentMode=RCTM_User, time=0.6, value=0.9}
           {arriveTangent=-3.5635, interpMode=RCIM_Cubic, leaveTangent=-3.5635, tangentMode=RCTM_User, time=1, value=0}
          ]}}}
       CurveIndex=$System.NormalizedLoopAge
      )
     )
    )
    Rotation Vector=<0 0 1>
   )
   &Color(Color=&MultiplyLinearColor(
     Linear Color=&ColorFromCurve(
      ColorCurve=DataInterface() {propertyValues {
        AlphaCurve {keys [<at 0.0028 0>,<at 0.8642 1>,<at 1 0>]}
        BlueCurve {keys [<at 0.8368 3>]}
        GreenCurve {keys [<at 0.8368 3>]}
        RedCurve {keys [<at 0.8368 3>]}
       }}
      CurveIndex=&Max_Float(A:NiagaraFloat=$Particles.NormalizedAge, B:NiagaraFloat=$System.NormalizedLoopAge)
     )
     Scale Factor=<1 1 1 0.2>
    ))
  ]
  rendererClasses [&NiagaraMeshRendererProperties]
  renderers [{rendererClass=&NiagaraMeshRendererProperties, rendererIndex=0}]
 }
 NE_EnrgyGlow {
  emitterSpawnScript [/SetVariables() {isSetParameters=true}]
  emitterUpdateScript [
   &EmitterState(
    Inactive Response=::NewEnumerator1
    Life Cycle Mode=::NewEnumerator1
    Loop Behavior=::NewEnumerator1
    Loop Duration Mode=::NewEnumerator1
    UseLoopDelay=false
   )
   &SpawnBurst_Instantaneous(Use Spawn Probability=false, UseLoopCountLimit=false)
   /SetVariables(Emitter.CustomTimer:NiagaraFloat=&Add_Float(
     A=$Emitter.CustomTimer
     B=&Multiply_Float(A=0.001, B=&Lerp_Float(A=1, B=5, Alpha=$System.TimeDilatationAlpha))
    )) {isSetParameters=true}
  ]
  particleSpawnScript [
   &InitializeParticle(
    Color Mode=::NewEnumerator0
    Lifetime=$System.EffectDuration
    Sprite Size Mode=::NewEnumerator3
    Uniform Sprite Size=300
    UsePositionOffset=true
    Write Material Random=true
   )
   &InitialMeshOrientation(Mesh Orientation Mode=::NewEnumerator7, Rotation=<0 0.25 0>, Use Rotation Vector=true)
  ]
  particleUpdateScript [
   &ParticleState()
   &SocketLocation(
    Orientation Sampling=::NewEnumerator1
    SamplePosOffset=true
    Socket Reader=$System.Socket Reader
    Socket Sampling Mode=::NewEnumerator3
    VelocitySpeedLimit=false
   )
   &Color(Color=&MakeLinearColorFromVectorAndFloat(
     "Vector (RGB)"=&MakeVectorFromLinearColorRGB(Color=<rgba 0.006 0.393 0.8 1>)
     "Float (Alpha)"=&FloatFromCurve(FloatCurve=DataInterface() {propertyValues {Curve {keys [<at 0 0>,<at 0.85 1>,<at 1 0>]}}}, Scale Curve=0.25)
    ))
   &ScaleSpriteSize(Uniform Curve Sprite Scale=DataInterface() {propertyValues {Curve {keys [
        {arriveTangent=-0.71, interpMode=RCIM_Cubic, leaveTangent=-0.71, tangentMode=RCTM_User, time=0, value=1.5}
        {arriveTangent=-0.5263, interpMode=RCIM_Cubic, leaveTangent=-0.5263, time=0.8, value=1}
        {interpMode=RCIM_Cubic, tangentMode=RCTM_User, time=0.95, value=1}
        {arriveTangent=-0.9468, interpMode=RCIM_Cubic, leaveTangent=-0.9468, tangentMode=RCTM_User, time=1, value=0.5}
       ]}}})
   &FN_Sprite_Attributes(
    Color=&MultiplyLinearColor(Linear Color=$Particles.Color, Scale Factor=&MakeVector4_FromLinearColor(LinearColor=<rgba 1 1 1 0.4>))
    Create Color Attribute=true
    Create SpriteSize Attribute=true
    SpriteSize=&Multiply_Vector2D_ByFloat(Float=1.4, Vector2D=$Particles.SpriteSize)
    Write to Module Namespace=true
   )
   &FN_Sprite_Attributes(
    Color=&MultiplyLinearColor(
     Linear Color=$Particles.Color
     Scale Factor=&MakeVector4_FromLinearColor(LinearColor=<rgba 0.02 0.02 0.02 1>)
    )
    Create Color Attribute=true
    Create SpriteSize Attribute=true
    SpriteSize=&Multiply_Vector2D_ByFloat(Float=1.2, Vector2D=$Particles.SpriteSize)
    Write to Module Namespace=true
   ) {module=FN_Sprite_Attributes001}
   /SetVariables(Particles.EnergySpriteDynMat:Vector4f=&MakeVector4(X=&Multiply_Float(A=$Emitter.CustomTimer, B=3), Y=0, Z=0, W=0)) {isSetParameters=true}
   &FN_Mesh_Attributes(
    Color=&Multiply_LinearColorByFloat(
     Linear Color=&ColorFromCurve(ColorCurve=DataInterface() {propertyValues {
        AlphaCurve {keys [<at 0.0107 0>,<at 0.8649 0.25>,<at 1 0>]}
        BlueCurve {keys [<at 0.2055 1 RCIM_Cubic>]}
        GreenCurve {keys [<at 0.2055 1 RCIM_Cubic>]}
        RedCurve {keys [<at 0.2055 1 RCIM_Cubic>]}
       }})
     Float=0.5
    )
    Create Color Attribute=true
    Create Scale Attribute=true
    Scale=&VectorFromFloat(Value=&Divide_Float(A=&MakeFloatFromVector2d(Vector 2D=$Particles.SpriteSize), B=130))
    Write to Module Namespace=true
   )
   &FN_Light_Attributes(
    Color=<rgba 0.0339 0.3528 1 1>
    Create Color Attribute=true
    Create Exponent Attribute=true
    Create Position Attribute=true
    Create Radius Attribute=true
    Exponent=1
    Position=&AddVectorToPosition(Position=$Engine.Emitter.SimulationPosition, Vector=<0 0 80>)
    Radius=90
    Scale Color=&Lerp_Float(
     A=0
     B=20
     Alpha=&FloatFromCurve(FloatCurve=DataInterface() {propertyValues {Curve {keys [<at 0 0>,<at 0.95 1>,<at 1 0>]}}})
    )
   )
  ]
  rendererClasses [
   &NiagaraSpriteRendererProperties
   &NiagaraLightRendererProperties
   &NiagaraMeshRendererProperties
  ]
  renderers [
   {rendererClass=&NiagaraSpriteRendererProperties}
   {rendererClass=&NiagaraSpriteRendererProperties, rendererIndex=1}
   {rendererClass=&NiagaraLightRendererProperties, rendererIndex=2}
   {rendererClass=&NiagaraSpriteRendererProperties, rendererIndex=3}
   {rendererClass=&NiagaraMeshRendererProperties, rendererIndex=4}
  ]
 }
 NE_Sparks {
  emitterSpawnScript [/SetVariables() {isSetParameters=true}]
  emitterUpdateScript [
   &EmitterState(
    Inactive Response=::NewEnumerator1
    Life Cycle Mode=::NewEnumerator1
    Loop Behavior=::NewEnumerator1
    Loop Duration Mode=::NewEnumerator1
    UseLoopDelay=false
   )
   &SpawnBurst_Instantaneous(Spawn Count=75, Use Spawn Probability=false, UseLoopCountLimit=false)
   /SetVariables(Emitter.CustomTimer:NiagaraFloat=&Add_Float(
     A=$Emitter.CustomTimer
     B=&Multiply_Float(A=0.001, B=&Lerp_Float(A=1, B=10, Alpha=$System.TimeDilatationAlpha))
    )) {isSetParameters=true}
  ]
  particleSpawnScript [
   &InitializeParticle(
    Color Mode=::NewEnumerator0
    Lifetime=&Multiply_Float(A=$System.EffectDuration, B=&RandomRangeFloat(Minimum=1, Maximum=1))
    Sprite Size=&Multiply_Vector2D_ByFloat(Float=&RandomRangeFloat(Minimum=2, Maximum=3), Vector2D=<1 2>)
    Sprite Size Mode=::NewEnumerator4
    UsePositionOffset=false
    Write Material Random=true
   )
   /SetVariables(Particles.Age:NiagaraFloat=&RandomRangeFloat(Minimum=-2, Maximum=0)) {isSetParameters=true}
   /SetVariables(Particles.InitialNormalizedIndex:NiagaraFloat=&ReturnNormalizedExecIndex()) {isSetParameters=true}
   &Color(Color=&ColorFromCurve(
     ColorCurve=DataInterface() {propertyValues {
       AlphaCurve {keys [<at 0 1 RCIM_Cubic>]}
       BlueCurve {keys [<at 0.4432 5 RCIM_Cubic>,<at 0.4849 5 RCIM_Cubic>]}
       GreenCurve {keys [<at 0.4432 0.9886 RCIM_Cubic>,<at 0.4849 4.0478 RCIM_Cubic>]}
       RedCurve {keys [<at 0.4432 0.0117 RCIM_Cubic>,<at 0.4849 0 RCIM_Cubic>]}
      }}
     CurveIndex=&RandomRangeFloat(Minimum=0, Maximum=1)
    ))
  ]
  particleUpdateScript [
   &ParticleState()
   &ScaleSpriteSize(Uniform Curve Sprite Scale=DataInterface() {propertyValues {Curve {keys [{tangentMode=RCTM_User, time=0, value=0}, {tangentMode=RCTM_User, time=0.15, value=1}]}}})
   &RotateAroundPoint(
    Radius:NiagaraFloat=&Multiply_Float(
     A=&RandomRangeFloat(Minimum=70, Maximum=95)
     B=&Multiply_Float(
      A=&FloatFromCurve(FloatCurve=DataInterface() {propertyValues {Curve {keys [
           {arriveTangent=-0.3661, interpMode=RCIM_Cubic, leaveTangent=-0.3661, tangentMode=RCTM_User, time=0, value=1.25}
           {interpMode=RCIM_Cubic, leaveTangent=-0.2, time=1, value=1}
          ]}}})
      B=&FloatFromCurve(
       FloatCurve=DataInterface() {propertyValues {Curve {keys [
           {arriveTangent=0.2913, interpMode=RCIM_Cubic, leaveTangent=0.2913, tangentMode=RCTM_User, time=0, value=1}
           {arriveTangent=-1, interpMode=RCIM_Cubic, leaveTangent=-1, time=0.85, value=0.8}
           {arriveTangent=-3.5635, interpMode=RCIM_Cubic, leaveTangent=-3.5635, tangentMode=RCTM_User, time=1, value=0}
          ]}}}
       CurveIndex=$System.NormalizedLoopAge
      )
     )
    )
    Rotation Center:NiagaraPosition=&AddVectorToPosition(Position=$Engine.Owner.Position, Vector=$System.SystemOffset)
    Rotation Phase:NiagaraFloat=&Add_Float(
     A=&Add_Float(A=$Particles.InitialNormalizedIndex, B=&Multiply_Float(A=$Emitter.CustomTimer, B=1))
     B=&Lerp_Float(
      A=0
      B=&RandomRangeFloat(Minimum=0.1, Maximum=0.5)
      Alpha=&FloatFromCurve(
       FloatCurve=DataInterface() {propertyValues {Curve {keys [
           {arriveTangent=0.3387, interpMode=RCIM_Cubic, leaveTangent=0.3387, tangentMode=RCTM_User, time=0.5, value=0}
           {arriveTangent=3.5279, interpMode=RCIM_Cubic, leaveTangent=3.5279, tangentMode=RCTM_User, time=1, value=1}
          ]}}}
       CurveIndex=$System.NormalizedLoopAge
      )
     )
    )
    X Axis:Vector3f=&RotateVectorByQuaternion(
     Vector To Rotate=<1 0 0>
     Quaternion=&MakeQuaternion(
      Quaternion From=0
      Axis=<0 1 0>
      Angle=&RandomRangeFloat(Minimum=-0.05, Maximum=0.05)
      From=?
      To=?
      Angle Type=::NewEnumerator1
      Normalize Vectors=false
     )
    )
    Y Axis:Vector3f=&RotateVectorByQuaternion(
     Vector To Rotate=<0 1 0>
     Quaternion=&MakeQuaternion(
      Quaternion From=0
      Axis=<1 0 0>
      Angle=&RandomRangeFloat(Minimum=-0.05, Maximum=0.05)
      From=?
      To=?
      Angle Type=::NewEnumerator1
      Normalize Vectors=false
     )
    )
   )
   &FN_Sprite_Attributes(
    Color=&MultiplyLinearColor(
     Linear Color=$Particles.Color
     Scale Factor=&MakeVector4_FromLinearColor(LinearColor=<rgba 0.5 0.5 0.5 0.01>)
    )
    Create Color Attribute=true
    Create SpriteSize Attribute=true
    SpriteSize=&Multiply_Vector2D(A=$Particles.SpriteSize, B=<10 3>)
    Write to Module Namespace=true
   )
   &SolveForcesAndVelocity(Clamp Velocity=false, Limit Acceleration=false)
   &SpriteFacingAndAlignment(
    Sprite Alignment=&NormalizeVector(
     Vector To Normalize=&Subtract_Position(A=$Particles.Position, B=$Particles.Previous.Position)
     Safe Normalize=true
     Safe Fallback Threshold=0
     Safe Fallback Vector=<0 0 0.001>
    )
    WriteAlignment=true
    WriteFacing=true
   )
   &ScaleColor(
    Scale Alpha=&Multiply_Float(
     A=&FloatFromCurve(FloatCurve=DataInterface() {propertyValues {Curve {keys [<at 0 0>,<at 1 1>]}}})
     B=&FloatFromCurve(FloatCurve=DataInterface() {propertyValues {Curve {keys [<at 0.7 1>,<at 1 0>]}}}, CurveIndex=$System.NormalizedLoopAge)
    )
    ScaleA=true
    ScaleRGB=true
   )
  ]
  rendererClasses [&NiagaraSpriteRendererProperties]
  renderers [
   {rendererClass=&NiagaraSpriteRendererProperties}
   {rendererClass=&NiagaraSpriteRendererProperties, rendererIndex=1}
  ]
 }
}
summary {
 systemName=NS_Emote_SonnyBoy_WindUp
 emitters [
  {emitterName=NE_Sparks, rendererClasses [&NiagaraSpriteRendererProperties]}
  {
   emitterName=NE_EnrgyGlow
   rendererClasses [
    &NiagaraSpriteRendererProperties
    &NiagaraLightRendererProperties
    &NiagaraMeshRendererProperties
   ]
  }
  {emitterName=NE_EnergyGuts, rendererClasses [&NiagaraMeshRendererProperties]}
 ]
 userVariables [User.BubbleOpacity {type=NiagaraFloat, defaultValue=0.5, description=""}]
}
